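# Material function that supplies Primary, Secondary and Tertiary colors for Mojo stage show effects.
# Each output is an RGB value from the MPC_Sparks parameter collection, or from manual vector parameters when the "Use Manual Colors" static switch is on.
# An unconnected branch samples three consecutive Mojo_StageShow_CurveAtlas rows using StageShowPalette and CustomUPosition, so neither affects the outputs.
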MaterialFunction /SparksStageShow/Systems/StageManager/FXManager/MaterialFunctions/MF_MojoStageColors
graph {nodes [
  ComponentMask(None=TextureCoordinate()) {r=true}
  n1 = VectorParameter() {group=Use Manual Color, parameterName="Primary Color (Manual)", defaultValue=<rgba 0 0.9647 0.051 1>}
  n2 = CollectionParameter() {collection=/SparksEffects/Effects/MPC_Sparks, parameterName=StageShowPrimary}
  FunctionOutput(None=ComponentMask(None=StaticSwitchParameter(True=@n1.RGB, False=@n2) {group=Use Manual Color, parameterName=Use Manual Colors}) {b=true, g=true, r=true}) {outputName=Primary}
  n3 = VectorParameter() {group=Use Manual Color, parameterName="Secondary Color (Manual)", defaultValue=<rgba 0 0.3333 1 1>}
  n4 = CollectionParameter() {collection=&MPC_Sparks, parameterName=StageShowSecondary}
  FunctionOutput(None=ComponentMask(None=StaticSwitchParameter(True=@n3.RGB, False=@n4) {group=Use Manual Color, parameterName=Use Manual Colors}) {b=true, g=true, r=true}) {outputName=Secondary, sortPriority=1}
  n5 = VectorParameter() {group=Use Manual Color, parameterName="Tertiary Color (Manual)", defaultValue=<rgba 0.7294 0 1 1>}
  n6 = CollectionParameter() {collection=&MPC_Sparks, parameterName=StageShowTertiary}
  FunctionOutput(None=ComponentMask(None=StaticSwitchParameter(True=@n5.RGB, False=@n6) {group=Use Manual Color, parameterName=Use Manual Colors}) {b=true, g=true, r=true}) {outputName=Tertiary, sortPriority=2}
  n7 = CollectionParameter() {collection=&MPC_Sparks, parameterName=StageShowPalette}
  Multiply(A=@n7) {constB=3}
  n8 = FunctionInput() {bUsePreviewValueAsDefault=true, inputName=CustomUPosition, inputType=FunctionInput_Scalar}
  n9 = Add(A=Add(A=NamedRerouteDeclaration(None=@n7) {name=Palette, nodeColor=<rgba 0 1 0.0156 1>}, B=Constant() {desc=Row}) {constB=0}) {constB=0.5}
  n10 = TextureObject() {samplerType=SAMPLERTYPE_LinearColor, texture=/SparksStageShow/Systems/StageManager/FXManager/Curves/Mojo_StageShow_CurveAtlas}
  n11 = NamedRerouteDeclaration(None=ComponentMask(None=TextureProperty(Tex=@n10)) {g=true}) {name=TextureSize, nodeColor=<rgba 0.8047 1 0 1>}
  TextureSample!(
   UVs=AppendVector(A=NamedRerouteDeclaration(None=@n8) {name=CustomUPosition, nodeColor=<rgba 1 0.0709 0 1>}, B=Divide(A=@n9, B=@n11) {constB=60})
   Tex=NamedRerouteDeclaration(None=@n10) {name=TextureObect, nodeColor=<rgba 1 0 0.759 1>}
   Level
   Apply View MipBias
  ) {mipValueMode=TMVM_MipLevel}
  n12 = Add(A=Add(A=NamedRerouteUsage(), B=Constant() {desc=Row, r=1}) {constB=0}) {constB=0.5}
  n13 = NamedRerouteUsage()
  TextureSample(UVs=AppendVector(A=NamedRerouteUsage(), B=Divide(A=@n12, B=@n13) {constB=60}), Tex=NamedRerouteUsage()) {mipValueMode=TMVM_MipLevel}
  n14 = Add(A=Add(A=NamedRerouteUsage(), B=Constant() {desc=Row, r=2}) {constB=0}) {constB=0.5}
  n15 = NamedRerouteUsage()
  TextureSample(UVs=AppendVector(A=NamedRerouteUsage(), B=Divide(A=@n14, B=@n15) {constB=60}), Tex=NamedRerouteUsage()) {mipValueMode=TMVM_MipLevel}
 ]}
